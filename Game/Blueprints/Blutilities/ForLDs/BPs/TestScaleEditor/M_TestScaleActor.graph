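# Masked, lit material for the level designers' test scale actor, usable on skeletal meshes.
# Body textures supply color and normal, with the mask's blue channel shared by roughness, metallic and specular.
# Camera distance divided by CameraDist drives opacity mask, clipping pixels near the camera.
# MF_RimV3 processes the attributes, then EyeAdaptationInverse compensates emissive for exposure; the HideLocation transform is left unconnected.
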Material /Game/Blueprints/Blutilities/ForLDs/BPs/TestScaleEditor/M_TestScaleActor
graph {
 properties {
  bUseMaterialAttributes=true
  bUsedWithMeshDeformer=true
  bUsedWithSkeletalMesh=true
  blendMode=BLEND_Masked
  pixelDepthOffsetMode=PDOM_Legacy
  usedShadingModels=Default Lit
 }
 nodes [
  n1 = Distance(A=CameraPositionWS(), B=WorldPosition().XYZ)
  n2 = ScalarParameter() {defaultValue=100, parameterName=CameraDist}
  n3 = TextureSampleParameter2D() {
   parameterName=Normal
   samplerType=SAMPLERTYPE_Normal
   texture=&/T_TestScaleActor_Body_N
  }
  n4 = TextureSampleParameter2D() {parameterName=Masks, samplerType=SAMPLERTYPE_Masks, texture=&/T_TestScaleActor_Body_S}
  n5 = SetMaterialAttributes!(
   MaterialAttributes
   Base Color=TextureSampleParameter2D() {parameterName=Base Color, texture=&/T_TestScaleActor_Body_D}.RGB
   Opacity Mask=Saturate(None=Divide(A=@n1, B=@n2))
   Normal=@n3.RGB
   Roughness=@n4.B
   Metallic=@n4.B
   Specular=@n4.B
  ) {attributeSetTypes [
    #1
    #2
    #3
    #4
    #5
    #6
   ]}
  n6 = GetMaterialAttributes(None=MaterialFunctionCall!(Material=@n5, Pre Skinned Local Position, AO=ScalarParameter() {defaultValue=0.15, parameterName=AO}, Gmap, Distance Blend=Constant() {r=1}) {materialFunction=/Game/Packages/Fortress_SharedMaterials/Base_Material_Functions/MF_RimV3}.Result) {attributeGetTypes [#7]}
  SetMaterialAttributes!(MaterialAttributes=@n6.MaterialAttributes, Emissive Color=EyeAdaptationInverse(LightValueInput=@n6.Emissive Color).EyeAdaptationInverse) {attributeSetTypes [#7]}
  TransformPosition(None=VectorParameter() {parameterName=HideLocation, defaultValue=<rgba 0 0 0 1>}.RGB) {transformType=TRANSFORMPOSSOURCE_World}
 ]
}
