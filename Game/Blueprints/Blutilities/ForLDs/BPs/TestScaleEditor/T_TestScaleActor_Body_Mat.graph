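# Masked, default-lit material for the body of a test scale actor, which by its folder path belongs to a scale-testing editor tool for level designers.
# It holds three texture samples: a `_D` texture, an `_S` texture sampled as masks, and an `_N` normal map.
# No math nodes or parameters appear, and automatic usage flagging in the editor is turned off.
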Material /Game/Blueprints/Blutilities/ForLDs/BPs/TestScaleEditor/T_TestScaleActor_Body_Mat
graph {
 properties {bAutomaticallySetUsageInEditor=false, blendMode=BLEND_Masked, usedShadingModels=Default Lit}
 nodes [
  TextureSample() {texture=&/T_TestScaleActor_Body_D}
  TextureSample() {samplerType=SAMPLERTYPE_Masks, texture=&/T_TestScaleActor_Body_S}
  TextureSample() {samplerType=SAMPLERTYPE_Normal, texture=&/T_TestScaleActor_Body_N}
 ]
}
