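# Animation Blueprint for the driver character in the level designers' TestScaleEditor tool, posing a Fortnite medium player skeleton as standing or crouched.
# On initialization only, it casts the owning actor to TestScaleEditorActor and copies that actor's isCrouched flag.
# The AnimGraph uses that flag in a linear Blend Poses by bool to choose between the assault rifle relaxed crouch pose and patrol idle pose.
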AnimBlueprint /Game/Blueprints/Blutilities/ForLDs/BPs/TestScaleEditor/TestScaleEditor_Driver_AnimBP
target_skeleton=&Fortnite_M_Avg_Player_Skeleton
graphs [
 AnimGraph {nodes [Output Pose(Result=Blend Poses by bool(
     BlendPose_0=Sequence Player 'AR_Crouch_Relaxed_Pose_CMM'() {sequence=/Game/Animation/Game/MainPlayer/Poses/Medium/Male/AssaultRifle/AR_Crouch_Relaxed_Pose_CMM}
     BlendPose_1=Sequence Player 'AssaultRifle_Idle_Patrol_Pose'() {sequence=/Game/Animation/Game/MainPlayer/Locomotion/Medium/Male/Idle/Patrol/AssaultRifle_Idle_Patrol_Pose}
     bActiveValue=Get isCrouched() -> (isCrouched:bool)
    ) {blendType=Linear}) {layerGroup=None}]}
 EventGraph {nodes [
   Event BlueprintInitializeAnimation()
   > n1 = Cast To TestScaleEditorActor!(execute, Object=GetOwningActor()) -> (then, CastFailed, AsTest Scale Editor Actor:Test Scale Editor Actor Object Reference) {is=title K2Node_DynamicCast}
   > Set isCrouched(isCrouched:bool=Get isCrouched!(self:Test Scale Editor Actor Object Reference=@n1) -> (isCrouched)) -> (Output_Get:bool)
  ]}
]
variables [isCrouched {type=boolean}]
